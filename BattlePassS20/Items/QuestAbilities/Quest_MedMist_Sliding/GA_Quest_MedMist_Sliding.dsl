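; Quest gameplay ability that credits progress for a Med Mist while-sliding quest, judging by its name.
; On activation it caches the avatar's Fortnite pawn and quest manager, then tests the pawn's owned gameplay tags against TagsToCheck.
; A match sends a complex custom stat event with StatEventSourceTags; either result ends the ability.
; An invalid pawn or controller only prints an error, without ending the ability.
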(event Ability|EventFailedToActivatePassiveAbility
  (params (ActorInfo :type Gameplay%20Ability%20Actor%20Info%20Structure)))

(event Ability|EventActivateAbility
  (bind _returnvalue (CallFunction|GetAvatarasFortPlayerPawn))
  (|CallParentFunction)
  (Utilities|IsValid _returnvalue
    (:Is%20Valid
      (bind _output_get (Variables|GAVars|Data|SetOwningPlayerPawn _returnvalue))
      (Utilities|IsValid (Fortnite|GetFortPlayerControllerfromActor _output_get)
        (:Is%20Valid
          (Variables|GAVars|Data|SetOwningPlayerQuestManager (Quests|GetQuestManager (Fortnite|GetFortPlayerControllerfromActor _output_get) "Invalid"))
          (bind _returnvalue_2 (MissionHelpers|HasTags (GameplayTags|GetOwnedGameplayTags (Variables|GAVars|Data|GetOwningPlayerPawn) :!class /Script/GameplayTags.GameplayTagAssetInterface) (Variables|Default|GetTagsToCheck) false false))
          (if _returnvalue_2
            (Quest|SendComplexCustomStatEvent (Variables|GAVars|Data|GetOwningPlayerQuestManager) 0 (Variables|Default|GetStatEventSourceTags) "(GameplayTags=)" 1)
            (Ability|EndAbility)
            (else
              (Ability|EndAbility))))
        (:Is%20Not%20Valid
          (Development|PrintString "@@@ ActivateAbilityFromEvent [GA_Quest_HealSpray_Sliding] GET PLAYER CONTROLLER FAILED"))))
    (:Is%20Not%20Valid
      (Development|PrintString "@@@ ActivateAbilityFromEvent [GA_Quest_HealSpray_Sliding]: CAST TO FORT PAWN FAILED"))))
